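; LF_2x5s_Parent is a parent Blueprint for layout foundations in the world tile set, apparently for a 2x5 variant.
; Its construction script configures the foundation's text render component as a label.
; The label's text comes from the EditorDisplayNameText variable and its world size from TextSize, so it refreshes whenever construction reruns.
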(fn ConstructionScript
  (bind _rendertextcomponent (Variables|Foundation|GetRenderTextComponent))
  (Rendering|Components|TextRender|SetText _rendertextcomponent (Variables|Foundation|GetEditorDisplayNameText))
  (Rendering|Components|TextRender|SetWorldSize _rendertextcomponent (Variables|Foundation|GetTextSize)))
